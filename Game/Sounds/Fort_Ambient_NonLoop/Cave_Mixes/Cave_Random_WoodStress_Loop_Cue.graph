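# Sound cue for cave ambience that plays sporadic wood-stress one-shots as an endless, spatialized loop in the non-looping ambient sound class.
# A quality-level switch picks one of four branches, each chaining Looping, random Delay, pitch/volume Modulator and Random wave selection.
# The variation pool shrinks from thirteen waves to three, then to two "_opt" waves in the last two branches.
# Overridden concurrency allows one instance, stopping the farthest first.
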SoundCue /Game/Sounds/Fort_Ambient_NonLoop/Cave_Mixes/Cave_Random_WoodStress_Loop_Cue
root=QualityLevel_0
nodes [
 {id=Delay_0, children [Modulator_0], properties {delayMax=10, delayMin=4}}
 {id=Delay_1, children [Modulator_1], properties {delayMax=10, delayMin=4}}
 {id=Delay_2, children [Modulator_2], properties {delayMax=10, delayMin=4}}
 {id=Delay_3, children [Modulator_3], properties {delayMax=10, delayMin=4}}
 {id=Looping_0, children [Delay_0]}
 {id=Looping_1, children [Delay_1]}
 {id=Looping_2, children [Delay_2]}
 {id=Looping_3, children [Delay_3]}
 {id=Modulator_0, children [Random_3], properties {pitchMax=1.25, pitchMin=0.7, volumeMax=1.1, volumeMin=0.7}}
 {id=Modulator_1, children [Random_1], properties {pitchMax=1.1, pitchMin=0.9, volumeMin=0.9}}
 {id=Modulator_2, children [Random_0], properties {pitchMax=1.1, pitchMin=0.6, volumeMin=0.9}}
 {id=Modulator_3, children [Random_4], properties {pitchMax=1.25, pitchMin=0.7, volumeMax=1.1, volumeMin=0.7}}
 {id=QualityLevel_0, children [Looping_1,Looping_2,Looping_0,Looping_3]}
 {id=Random_0, children [WavePlayer_2,WavePlayer_5,WavePlayer_8], properties {weights [1,1,1]}}
 {
  id=Random_1
  children [
   WavePlayer_15
   WavePlayer_1
   WavePlayer_17
   WavePlayer_7
   WavePlayer_19
   WavePlayer_20
   WavePlayer_21
   WavePlayer_4
   WavePlayer_23
   WavePlayer_24
   WavePlayer_25
   WavePlayer_26
   WavePlayer_27
  ]
  properties {weights [1,1,1,1,1,1,1,1,1,1,1,1,1]}
 }
 {id=Random_3, children [WavePlayer_0,WavePlayer_3], properties {weights [1,1]}}
 {id=Random_4, children [WavePlayer_11, WavePlayer_12], properties {weights [1,1]}}
 {id=WavePlayer_0, properties {soundWaveAssetPtr=&/Cave_Random_WoodStress17_opt}}
 {id=WavePlayer_1, properties {soundWaveAssetPtr=&/Cave_Random_WoodStress17}}
 {id=WavePlayer_11, properties {soundWaveAssetPtr=&Cave_Random_WoodStress17_opt}}
 {id=WavePlayer_12, properties {soundWaveAssetPtr=&/Cave_Random_WoodStress19_opt}}
 {id=WavePlayer_15, properties {soundWaveAssetPtr=&/Cave_Random_WoodStress16}}
 {id=WavePlayer_17, properties {soundWaveAssetPtr=&/Cave_Random_WoodStress18}}
 {id=WavePlayer_19, properties {soundWaveAssetPtr=&/Cave_Random_WoodStress20}}
 {id=WavePlayer_2, properties {soundWaveAssetPtr=&Cave_Random_WoodStress17}}
 {id=WavePlayer_20, properties {soundWaveAssetPtr=&/Cave_Random_WoodStress23}}
 {id=WavePlayer_21, properties {soundWaveAssetPtr=&/Cave_Random_WoodStress24}}
 {id=WavePlayer_23, properties {soundWaveAssetPtr=&/Cave_Random_WoodStress27}}
 {id=WavePlayer_24, properties {soundWaveAssetPtr=&/Cave_Random_WoodStress30}}
 {id=WavePlayer_25, properties {soundWaveAssetPtr=&/Cave_Random_WoodStress33}}
 {id=WavePlayer_26, properties {soundWaveAssetPtr=&/Cave_Random_WoodStress36}}
 {id=WavePlayer_27, properties {soundWaveAssetPtr=&/Cave_Random_WoodStress37}}
 {id=WavePlayer_3, properties {soundWaveAssetPtr=&Cave_Random_WoodStress19_opt}}
 {id=WavePlayer_4, properties {soundWaveAssetPtr=&/Cave_Random_WoodStress25}}
 {id=WavePlayer_5, properties {soundWaveAssetPtr=&Cave_Random_WoodStress25}}
 {id=WavePlayer_7, properties {soundWaveAssetPtr=&/Cave_Random_WoodStress19}}
 {id=WavePlayer_8, properties {soundWaveAssetPtr=&Cave_Random_WoodStress19}}
]
properties {
 bOverrideAttenuation=true
 bOverrideConcurrency=true
 maxDistance=1500
 soundClassObject=/Game/Sounds/FortSoundClassesAndModes/Fort_Ambient_NonLoop
 volumeMultiplier=0.65
 attenuationOverrides {falloffDistance=1000, attenuationShapeExtents {x=500}}
 concurrencyOverrides {maxCount=1, resolutionRule=StopFarthestThenPreventNew}
}
